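# Physics asset for the F_MED_Spartan_Future female player body cosmetic, covering its belt and side-pouch dyn joints.
# Capsule bodies on the front, back and side belt joints, side pouch, pelvis and right thigh are each constrained to the pelvis.
# The pouch and front belt swing within limits with twist locked, and the back belt locks one swing axis.
# The side belt is angularly locked, while the thigh rotates freely.
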PhysicsAsset /BRCosmetics/Characters/Player/Female/Medium/Bodies/F_MED_Spartan_Future/Meshes/F_MED_Spartan_Future_Physics
bodies {
 dyn_beltBack_jnt [{shapeType=Capsule, length=12.6289, radius=6, shapeName=dyn_beltBack_jnt_capsule, center=<-1.4385 0 0>, rotation=<rot 0 0 90>}]
 dyn_beltFront_jnt [{shapeType=Capsule, length=14, radius=4, shapeName=dyn_beltFront_jnt_capsule, center=<3.0998 0.2952 -0.3847>, rotation=<rot 0 0 90>}]
 dyn_beltSide_jnt [{shapeType=Capsule, length=10, radius=2.5, center=<-1.4022 -1.6057 0.0645>, rotation=<rot 0 -2.5522 -82.295>}]
 dyn_sidePouch_jnt [{shapeType=Capsule, length=2, radius=2, center=<1.4009 0 0>, rotation=<rot 90 0 0>}]
 pelvis [{shapeType=Capsule, length=20, radius=9, center=<-6.163 -1.0838 8.3269>, rotation=<rot 90 90 91.0755>}]
 thigh_r [{shapeType=Capsule, length=5, radius=2}]
}
constraints [
 {bone1Name=dyn_sidePouch_jnt, bone2Name=pelvis, swing1LimitDegrees=20, swing2LimitDegrees=30, twistLimitDegrees=5, twistMotion=Locked}
 {
  bone1Name=thigh_r
  bone2Name=pelvis
  swing1LimitDegrees=45
  swing1Motion=Free
  swing2LimitDegrees=45
  swing2Motion=Free
  twistLimitDegrees=45
  twistMotion=Free
 }
 {bone1Name=dyn_beltBack_jnt, bone2Name=pelvis, swing1LimitDegrees=45, swing1Motion=Locked, swing2LimitDegrees=5, twistLimitDegrees=5}
 {bone1Name=dyn_beltFront_jnt, bone2Name=pelvis, swing1LimitDegrees=20, swing2LimitDegrees=20, twistLimitDegrees=45, twistMotion=Locked}
 {
  bone1Name=dyn_beltSide_jnt
  bone2Name=pelvis
  swing1LimitDegrees=45
  swing1Motion=Locked
  swing2LimitDegrees=45
  swing2Motion=Locked
  twistLimitDegrees=45
  twistMotion=Locked
 }
]
