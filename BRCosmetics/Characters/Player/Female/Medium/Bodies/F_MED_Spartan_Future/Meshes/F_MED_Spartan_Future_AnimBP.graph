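# Post-processing AnimBP for the female medium Spartan Future cosmetic body, adding secondary motion and correctives to the pose copied from its attached parent.
# Multi-bone weighted constraints drive straps and belt bullets, while AnimDynamics and RigidBody simulate dangling gear, before a final Control Rig pass.
# EventGraph maps arm and wrist deltas from ref pose to alphas for forearm-volume and strap offsets.
# Crouch events blend in pouch and thigh-belt offsets.
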AnimBlueprint /BRCosmetics/Characters/Player/Female/Medium/Bodies/F_MED_Spartan_Future/Meshes/F_MED_Spartan_Future_AnimBP
target_skeleton=&/F_MED_Spartan_Future_Skeleton
graphs [
 AnimGraph {nodes [
   "Constraint - L_strapBack_B_jnt (PPPPPPPPPPPPPPP)"!(
    ComponentPose
    Alpha=1
    ConstraintWeights_0=0
    ConstraintWeights_1:float=0
    ConstraintWeights_2:float=0.501
    ConstraintWeights_3:float=0.0004
    ConstraintWeights_4:float=0
    ConstraintWeights_5:float=0
    ConstraintWeights_6:float=0.0019
    ConstraintWeights_7:float=0.0163
    ConstraintWeights_8:float=0.3389
    ConstraintWeights_9:float=0.0793
    ConstraintWeights_10:float=0.0073
    ConstraintWeights_11:float=0
    ConstraintWeights_12:float=0
    ConstraintWeights_13:float=0.0548
    ConstraintWeights_14:float=0
   ) -> (Pose) {
    alphaBoolBlend {blendOption=Linear}
    boneToModify {boneName=L_strapBack_B_jnt}
    constraintSetup [
     {targetBone {boneName=lowerarm_l}}
     {targetBone {boneName=upperarm_r}}
     {targetBone {boneName=clavicle_l}}
     {targetBone {boneName=upperarm_twist_02_l}}
     {targetBone {boneName=upperarm_twist_01_r}}
     {targetBone {boneName=spine_01}}
     {targetBone {boneName=spine_02}}
     {targetBone {boneName=spine_03}}
     {targetBone {boneName=spine_04}}
     {targetBone {boneName=spine_05}}
     {targetBone {boneName=upperarm_twist_01_l}}
     {targetBone {boneName=clavicle_r}}
     {targetBone {boneName=pelvis}}
     {targetBone {boneName=upperarm_l}}
     {targetBone {boneName=lowerarm_twist_02_l}}
    ]
    constraintWeights [1,1,1,1,1,1,1,1,1,1,1,1,1,1,1]
   }
   "Constraint - R_strapBack_A_jnt (PPPPPPPPPPPPPPPPPP)"!(
    ComponentPose
    Alpha=1
    ConstraintWeights_0=0
    ConstraintWeights_1:float=0
    ConstraintWeights_2:float=0.0414
    ConstraintWeights_3:float=0
    ConstraintWeights_4:float=0
    ConstraintWeights_5:float=0
    ConstraintWeights_6:float=0
    ConstraintWeights_7:float=0.0027
    ConstraintWeights_8:float=0.0001
    ConstraintWeights_9:float=0
    ConstraintWeights_10:float=0.0007
    ConstraintWeights_11:float=0.0038
    ConstraintWeights_12:float=0.1279
    ConstraintWeights_13:float=0.0657
    ConstraintWeights_14:float=0
    ConstraintWeights_15:float=0.7578
    ConstraintWeights_16:float=0
    ConstraintWeights_17:float=0
   ) -> (Pose) {
    alphaBoolBlend {blendOption=Linear}
    boneToModify {boneName=R_strapBack_A_jnt}
    constraintSetup [
     {targetBone {boneName=lowerarm_twist_02_r}}
     {targetBone {boneName=lowerarm_l}}
     {targetBone {boneName=upperarm_r}}
     {targetBone {boneName=upperarm_twist_02_l}}
     {targetBone {boneName=lowerarm_r}}
     {targetBone {boneName=clavicle_l}}
     {targetBone {boneName=lowerarm_twist_01_r}}
     {targetBone {boneName=upperarm_twist_01_r}}
     {targetBone {boneName=upperarm_twist_02_r}}
     {targetBone {boneName=spine_01}}
     {targetBone {boneName=spine_02}}
     {targetBone {boneName=spine_03}}
     {targetBone {boneName=spine_04}}
     {targetBone {boneName=spine_05}}
     {targetBone {boneName=upperarm_twist_01_l}}
     {targetBone {boneName=clavicle_r}}
     {targetBone {boneName=pelvis}}
     {targetBone {boneName=upperarm_l}}
    ]
    constraintWeights [1,1,1,1,1,1,1,1,1,1,1,1,1,1,1,1,1,1]
   }
   "Constraint - L_strapBack_A_jnt (PPPPPPPPPPPPPP)"!(
    ComponentPose
    Alpha=1
    ConstraintWeights_0=0
    ConstraintWeights_1:float=0
    ConstraintWeights_2:float=0.8104
    ConstraintWeights_3:float=0
    ConstraintWeights_4:float=0
    ConstraintWeights_5:float=0
    ConstraintWeights_6:float=0.0002
    ConstraintWeights_7:float=0.0002
    ConstraintWeights_8:float=0.0866
    ConstraintWeights_9:float=0.0225
    ConstraintWeights_10:float=0.0002
    ConstraintWeights_11:float=0
    ConstraintWeights_12:float=0
    ConstraintWeights_13:float=0.0798
   ) -> (Pose) {
    alphaBoolBlend {blendOption=Linear}
    boneToModify {boneName=L_strapBack_A_jnt}
    constraintSetup [
     {targetBone {boneName=lowerarm_l}}
     {targetBone {boneName=upperarm_r}}
     {targetBone {boneName=clavicle_l}}
     {targetBone {boneName=upperarm_twist_02_l}}
     {targetBone {boneName=upperarm_twist_01_r}}
     {targetBone {boneName=spine_01}}
     {targetBone {boneName=spine_02}}
     {targetBone {boneName=spine_03}}
     {targetBone {boneName=spine_04}}
     {targetBone {boneName=spine_05}}
     {targetBone {boneName=upperarm_twist_01_l}}
     {targetBone {boneName=clavicle_r}}
     {targetBone {boneName=pelvis}}
     {targetBone {boneName=upperarm_l}}
    ]
    constraintWeights [1,1,1,1,1,1,1,1,1,1,1,1,1,1]
   }
   n1 = "Constraint - L_strapBack_A_jnt (PPPPPPPPPPPPPPPPPPP)"!(
    ComponentPose=Local To Component(LocalPose=Copy Pose From Mesh() {bUseAttachedParent=true})
    Alpha=1
    ConstraintWeights_0=0
    ConstraintWeights_1:float=0
    ConstraintWeights_2:float=0
    ConstraintWeights_3:float=0.0001
    ConstraintWeights_4:float=0
    ConstraintWeights_5:float=0.6454
    ConstraintWeights_6:float=0
    ConstraintWeights_7:float=0
    ConstraintWeights_8:float=0
    ConstraintWeights_9:float=0
    ConstraintWeights_10:float=0.0021
    ConstraintWeights_11:float=0.0144
    ConstraintWeights_12:float=0.1399
    ConstraintWeights_13:float=0.1763
    ConstraintWeights_14:float=0.0018
    ConstraintWeights_15:float=0.0001
    ConstraintWeights_16:float=0
    ConstraintWeights_17:float=0.0198
    ConstraintWeights_18:float=0
   ) -> (Pose) {
    alphaBoolBlend {blendOption=Linear}
    boneToModify {boneName=L_strapBack_A_jnt}
    constraintSetup [
     {targetBone {boneName=lowerarm_twist_02_r}}
     {targetBone {boneName=lowerarm_l}}
     {targetBone {boneName=upperarm_r}}
     {targetBone {boneName=upperarm_twist_02_l}}
     {targetBone {boneName=lowerarm_r}}
     {targetBone {boneName=clavicle_l}}
     {targetBone {boneName=lowerarm_twist_01_r}}
     {targetBone {boneName=upperarm_twist_01_r}}
     {targetBone {boneName=upperarm_twist_02_r}}
     {targetBone {boneName=spine_01}}
     {targetBone {boneName=spine_02}}
     {targetBone {boneName=spine_03}}
     {targetBone {boneName=spine_04}}
     {targetBone {boneName=spine_05}}
     {targetBone {boneName=upperarm_twist_01_l}}
     {targetBone {boneName=clavicle_r}}
     {targetBone {boneName=pelvis}}
     {targetBone {boneName=upperarm_l}}
     {targetBone {boneName=lowerarm_twist_02_l}}
    ]
    constraintWeights [1,1,1,1,1,1,1,1,1,1,1,1,1,1,1,1,1,1,1]
   }
   n2 = "Constraint - L_strapBack_B_jnt (PPPPPPPPPPPPPPPPPPP)"!(
    ComponentPose=@n1
    Alpha=1
    ConstraintWeights_0=0
    ConstraintWeights_1:float=0
    ConstraintWeights_2:float=0
    ConstraintWeights_3:float=0.0001
    ConstraintWeights_4:float=0
    ConstraintWeights_5:float=0.389
    ConstraintWeights_6:float=0
    ConstraintWeights_7:float=0
    ConstraintWeights_8:float=0
    ConstraintWeights_9:float=0
    ConstraintWeights_10:float=0.0048
    ConstraintWeights_11:float=0.0388
    ConstraintWeights_12:float=0.4392
    ConstraintWeights_13:float=0.105
    ConstraintWeights_14:float=0.002
    ConstraintWeights_15:float=0.0001
    ConstraintWeights_16:float=0
    ConstraintWeights_17:float=0.021
    ConstraintWeights_18:float=0
   ) -> (Pose) {
    alphaBoolBlend {blendOption=Linear}
    boneToModify {boneName=L_strapBack_B_jnt}
    constraintSetup [
     {targetBone {boneName=lowerarm_twist_02_r}}
     {targetBone {boneName=lowerarm_l}}
     {targetBone {boneName=upperarm_r}}
     {targetBone {boneName=upperarm_twist_02_l}}
     {targetBone {boneName=lowerarm_r}}
     {targetBone {boneName=clavicle_l}}
     {targetBone {boneName=lowerarm_twist_01_r}}
     {targetBone {boneName=upperarm_twist_01_r}}
     {targetBone {boneName=upperarm_twist_02_r}}
     {targetBone {boneName=spine_01}}
     {targetBone {boneName=spine_02}}
     {targetBone {boneName=spine_03}}
     {targetBone {boneName=spine_04}}
     {targetBone {boneName=spine_05}}
     {targetBone {boneName=upperarm_twist_01_l}}
     {targetBone {boneName=clavicle_r}}
     {targetBone {boneName=pelvis}}
     {targetBone {boneName=upperarm_l}}
     {targetBone {boneName=lowerarm_twist_02_l}}
    ]
    constraintWeights [1,1,1,1,1,1,1,1,1,1,1,1,1,1,1,1,1,1,1]
   }
   n3 = "Constraint - R_strapBack_A_jnt (PPPPPPPPPPPPPPPP)"!(
    ComponentPose=@n2
    Alpha=1
    ConstraintWeights_0=0
    ConstraintWeights_1:float=0.0227
    ConstraintWeights_2:float=0
    ConstraintWeights_3:float=0.0001
    ConstraintWeights_4:float=0
    ConstraintWeights_5:float=0.0021
    ConstraintWeights_6:float=0.0001
    ConstraintWeights_7:float=0
    ConstraintWeights_8:float=0.0035
    ConstraintWeights_9:float=0.0258
    ConstraintWeights_10:float=0.2675
    ConstraintWeights_11:float=0.1258
    ConstraintWeights_12:float=0
    ConstraintWeights_13:float=0.5524
    ConstraintWeights_14:float=0
    ConstraintWeights_15:float=0
   ) -> (Pose) {
    alphaBoolBlend {blendOption=Linear}
    boneToModify {boneName=R_strapBack_A_jnt}
    constraintSetup [
     {targetBone {boneName=lowerarm_twist_02_r}}
     {targetBone {boneName=upperarm_r}}
     {targetBone {boneName=lowerarm_r}}
     {targetBone {boneName=clavicle_l}}
     {targetBone {boneName=upperarm_twist_02_l}}
     {targetBone {boneName=upperarm_twist_01_r}}
     {targetBone {boneName=upperarm_twist_02_r}}
     {targetBone {boneName=spine_01}}
     {targetBone {boneName=spine_02}}
     {targetBone {boneName=spine_03}}
     {targetBone {boneName=spine_04}}
     {targetBone {boneName=spine_05}}
     {targetBone {boneName=upperarm_twist_01_l}}
     {targetBone {boneName=clavicle_r}}
     {targetBone {boneName=pelvis}}
     {targetBone {boneName=upperarm_l}}
    ]
    constraintWeights [1,1,1,1,1,1,1,1,1,1,1,1,1,1,1,1]
   }
   n4 = "Constraint - R_strapBack_B_jnt (PPPPPPPPPPPPPPPP)"!(
    ComponentPose=@n3
    Alpha=1
    ConstraintWeights_0=0
    ConstraintWeights_1:float=0.021
    ConstraintWeights_2:float=0
    ConstraintWeights_3:float=0.0001
    ConstraintWeights_4:float=0
    ConstraintWeights_5:float=0.002
    ConstraintWeights_6:float=0.0001
    ConstraintWeights_7:float=0
    ConstraintWeights_8:float=0.0048
    ConstraintWeights_9:float=0.0388
    ConstraintWeights_10:float=0.4392
    ConstraintWeights_11:float=0.105
    ConstraintWeights_12:float=0
    ConstraintWeights_13:float=0.3891
    ConstraintWeights_14:float=0
    ConstraintWeights_15:float=0
   ) -> (Pose) {
    alphaBoolBlend {blendOption=Linear}
    boneToModify {boneName=R_strapBack_B_jnt}
    constraintSetup [
     {targetBone {boneName=lowerarm_twist_02_r}}
     {targetBone {boneName=upperarm_r}}
     {targetBone {boneName=lowerarm_r}}
     {targetBone {boneName=clavicle_l}}
     {targetBone {boneName=upperarm_twist_02_l}}
     {targetBone {boneName=upperarm_twist_01_r}}
     {targetBone {boneName=upperarm_twist_02_r}}
     {targetBone {boneName=spine_01}}
     {targetBone {boneName=spine_02}}
     {targetBone {boneName=spine_03}}
     {targetBone {boneName=spine_04}}
     {targetBone {boneName=spine_05}}
     {targetBone {boneName=upperarm_twist_01_l}}
     {targetBone {boneName=clavicle_r}}
     {targetBone {boneName=pelvis}}
     {targetBone {boneName=upperarm_l}}
    ]
    constraintWeights [1,1,1,1,1,1,1,1,1,1,1,1,1,1,1,1]
   }
   n5 = "Constraint - R_strapFront_A_jnt (PPPPPPPPPPPPPPPP)"!(
    ComponentPose=@n4
    Alpha=1
    ConstraintWeights_0=0
    ConstraintWeights_1:float=0.4106
    ConstraintWeights_2:float=0
    ConstraintWeights_3:float=0
    ConstraintWeights_4:float=0
    ConstraintWeights_5:float=0.0015
    ConstraintWeights_6:float=0
    ConstraintWeights_7:float=0
    ConstraintWeights_8:float=0.002
    ConstraintWeights_9:float=0.0043
    ConstraintWeights_10:float=0.2281
    ConstraintWeights_11:float=0.0375
    ConstraintWeights_12:float=0
    ConstraintWeights_13:float=0.316
    ConstraintWeights_14:float=0
    ConstraintWeights_15:float=0
   ) -> (Pose) {
    alphaBoolBlend {blendOption=Linear}
    boneToModify {boneName=R_strapFront_A_jnt}
    constraintSetup [
     {targetBone {boneName=lowerarm_twist_02_r}}
     {targetBone {boneName=upperarm_r}}
     {targetBone {boneName=lowerarm_r}}
     {targetBone {boneName=clavicle_l}}
     {targetBone {boneName=upperarm_twist_02_l}}
     {targetBone {boneName=upperarm_twist_01_r}}
     {targetBone {boneName=upperarm_twist_02_r}}
     {targetBone {boneName=spine_01}}
     {targetBone {boneName=spine_02}}
     {targetBone {boneName=spine_03}}
     {targetBone {boneName=spine_04}}
     {targetBone {boneName=spine_05}}
     {targetBone {boneName=upperarm_twist_01_l}}
     {targetBone {boneName=clavicle_r}}
     {targetBone {boneName=pelvis}}
     {targetBone {boneName=upperarm_l}}
    ]
    constraintWeights [1,1,1,1,1,1,1,1,1,1,1,1,1,1,1,1]
   }
   n6 = "Constraint - R_strapFront_B_jnt (PPPPPPPPPPPPPP)"!(
    ComponentPose=@n5
    Alpha=1
    ConstraintWeights_0=0.2143
    ConstraintWeights_1:float=0
    ConstraintWeights_2:float=0
    ConstraintWeights_3:float=0
    ConstraintWeights_4:float=0.0009
    ConstraintWeights_5:float=0
    ConstraintWeights_6:float=0.0001
    ConstraintWeights_7:float=0.007
    ConstraintWeights_8:float=0.0127
    ConstraintWeights_9:float=0.4063
    ConstraintWeights_10:float=0.1267
    ConstraintWeights_11:float=0
    ConstraintWeights_12:float=0.232
    ConstraintWeights_13:float=0
   ) -> (Pose) {
    alphaBoolBlend {blendOption=Linear}
    boneToModify {boneName=R_strapFront_B_jnt}
    constraintSetup [
     {targetBone {boneName=upperarm_r}}
     {targetBone {boneName=lowerarm_r}}
     {targetBone {boneName=clavicle_l}}
     {targetBone {boneName=upperarm_twist_02_l}}
     {targetBone {boneName=upperarm_twist_01_r}}
     {targetBone {boneName=upperarm_twist_02_r}}
     {targetBone {boneName=spine_01}}
     {targetBone {boneName=spine_02}}
     {targetBone {boneName=spine_03}}
     {targetBone {boneName=spine_04}}
     {targetBone {boneName=spine_05}}
     {targetBone {boneName=upperarm_twist_01_l}}
     {targetBone {boneName=clavicle_r}}
     {targetBone {boneName=pelvis}}
    ]
    constraintWeights [1,1,1,1,1,1,1,1,1,1,1,1,1,1]
   }
   n7 = "Constraint - L_strapFront_A_jnt (PPPPPPPPPPPPPPP)"!(
    ComponentPose=@n6
    Alpha=1
    ConstraintWeights_0=0
    ConstraintWeights_1:float=0.2937
    ConstraintWeights_2:float=0.0001
    ConstraintWeights_3:float=0
    ConstraintWeights_4:float=0
    ConstraintWeights_5:float=0
    ConstraintWeights_6:float=0.0012
    ConstraintWeights_7:float=0.0028
    ConstraintWeights_8:float=0.1962
    ConstraintWeights_9:float=0.0274
    ConstraintWeights_10:float=0.0037
    ConstraintWeights_11:float=0
    ConstraintWeights_12:float=0
    ConstraintWeights_13:float=0.4748
    ConstraintWeights_14:float=0
   ) -> (Pose) {
    alphaBoolBlend {blendOption=Linear}
    boneToModify {boneName=L_strapFront_A_jnt}
    constraintSetup [
     {targetBone {boneName=lowerarm_l}}
     {targetBone {boneName=clavicle_l}}
     {targetBone {boneName=upperarm_twist_02_l}}
     {targetBone {boneName=upperarm_twist_01_r}}
     {targetBone {boneName=upperarm_twist_02_r}}
     {targetBone {boneName=spine_01}}
     {targetBone {boneName=spine_02}}
     {targetBone {boneName=spine_03}}
     {targetBone {boneName=spine_04}}
     {targetBone {boneName=spine_05}}
     {targetBone {boneName=upperarm_twist_01_l}}
     {targetBone {boneName=clavicle_r}}
     {targetBone {boneName=pelvis}}
     {targetBone {boneName=upperarm_l}}
     {targetBone {boneName=lowerarm_twist_02_l}}
    ]
    constraintWeights [1,1,1,1,1,1,1,1,1,1,1,1,1,1,1]
   }
   n8 = "Constraint - L_strapFront_B_jnt (PPPPPPPPPPPPPPP)"!(
    ComponentPose=@n7
    Alpha=1
    ConstraintWeights_0=0
    ConstraintWeights_1:float=0.2239
    ConstraintWeights_2:float=0
    ConstraintWeights_3:float=0
    ConstraintWeights_4:float=0
    ConstraintWeights_5:float=0.0001
    ConstraintWeights_6:float=0.006
    ConstraintWeights_7:float=0.0106
    ConstraintWeights_8:float=0.3308
    ConstraintWeights_9:float=0.0827
    ConstraintWeights_10:float=0.0013
    ConstraintWeights_11:float=0
    ConstraintWeights_12:float=0
    ConstraintWeights_13:float=0.3447
    ConstraintWeights_14:float=0
   ) -> (Pose) {
    alphaBoolBlend {blendOption=Linear}
    boneToModify {boneName=L_strapFront_B_jnt}
    constraintSetup [
     {targetBone {boneName=lowerarm_l}}
     {targetBone {boneName=clavicle_l}}
     {targetBone {boneName=upperarm_twist_02_l}}
     {targetBone {boneName=upperarm_twist_01_r}}
     {targetBone {boneName=upperarm_twist_02_r}}
     {targetBone {boneName=spine_01}}
     {targetBone {boneName=spine_02}}
     {targetBone {boneName=spine_03}}
     {targetBone {boneName=spine_04}}
     {targetBone {boneName=spine_05}}
     {targetBone {boneName=upperarm_twist_01_l}}
     {targetBone {boneName=clavicle_r}}
     {targetBone {boneName=pelvis}}
     {targetBone {boneName=upperarm_l}}
     {targetBone {boneName=lowerarm_twist_02_l}}
    ]
    constraintWeights [1,1,1,1,1,1,1,1,1,1,1,1,1,1,1]
   }
   n9 = "Constraint - L_strapFront_center_B_jnt (PP)"!(ComponentPose=@n8, Alpha=1, ConstraintWeights_0=0.9988, ConstraintWeights_1:float=0.0012) -> (Pose) {
    alphaBoolBlend {blendOption=Linear}
    boneToModify {boneName=L_strapFront_center_B_jnt}
    constraintSetup [{targetBone {boneName=upperarm_l}}, {targetBone {boneName=upperarm_twist_01_l}}]
    constraintWeights [1,1]
   }
   n10 = "Constraint - R_strapFront_center_B_jnt (P)"!(ComponentPose=@n9, Alpha=1, ConstraintWeights_0=1, ConstraintWeights_1:float=0) -> (Pose) {
    alphaBoolBlend {blendOption=Linear}
    boneToModify {boneName=R_strapFront_center_B_jnt}
    constraintSetup [{targetBone {boneName=upperarm_r}}]
    constraintWeights [1,0]
   }
   n11 = Anim Dynamics - Bone: dyn_beltFront_jnt!(ComponentPose=@n10, Alpha=1, GravityScale:float=0.3, ExternalForce="0,0,0") -> (Pose) {
    bGravityOverrideInSimSpace=true
    gravityScale=0.3
    alphaBoolBlend {blendOption=Linear}
    boundBone {boneName=dyn_beltFront_jnt}
    componentAppliedLinearAccClamp=<4000 4000 4000>
    componentLinearAccScale=<0.9 0.9 0.9>
    physicsBodyDefinitions [{
      sphereCollisionRadius=1
      boundBone {boneName=dyn_beltFront_jnt}
      boxExtents=<5 5 10>
      constraintSetup {angularLimitsMax=<10 10 10>, angularLimitsMin=<-10 -10 -10>, angularTarget=<0 0 0>, linearAxesMax=<0 0 3>, linearAxesMin=<0 0 -1>}
      localJointOffset=<0 0 0>
     }]
    relativeSpaceBone {boneName=pelvis}
   }
   n12 = "Transform (Modify) Bone - Bone: dyn_L_forearm_volume"(ComponentPose=@n11, Alpha=Get L_forearm_ADS() -> (L_forearm_ADS:double), Translation="4.5,-0.5,0", Rotation="0,0,-5", Scale="0.9,1,1") {
    rotationMode=BMM_Additive
    rotationSpace=BCS_BoneSpace
    scaleMode=BMM_Additive
    scaleSpace=BCS_BoneSpace
    translationMode=BMM_Additive
    translationSpace=BCS_BoneSpace
    alphaBoolBlend {blendOption=Linear}
    boneToModify {boneName=dyn_L_forearm_volume}
    rotation {roll=-5}
    scale {x=0.9}
    translation=<4.5 -0.5>
   }
   n13 = "Transform (Modify) Bone - Bone: dyn_R_forearm_volume"(ComponentPose=@n12, Alpha=Get R_forearm_ADS() -> (R_forearm_ADS:double), Translation="4.5,-0.5,0", Rotation="0,0,-5", Scale="0.9,1,1") {
    rotationMode=BMM_Additive
    rotationSpace=BCS_BoneSpace
    scaleMode=BMM_Additive
    scaleSpace=BCS_BoneSpace
    translationMode=BMM_Additive
    translationSpace=BCS_BoneSpace
    alphaBoolBlend {blendOption=Linear}
    boneToModify {boneName=dyn_R_forearm_volume}
    rotation {roll=-5}
    scale {x=0.9}
    translation=<4.5 -0.5>
   }
   n14 = "Transform (Modify) Bone - Bone: dyn_R_forearmTwist_volume_2"(ComponentPose=@n13, Alpha=Get R_wrist_rotate_Y() -> (R_wrist_rotate_Y:double), Translation="-2,0,0", Rotation="0,0,0", Scale="0.6,1,1") {
    rotationMode=BMM_Additive
    rotationSpace=BCS_BoneSpace
    scaleMode=BMM_Additive
    scaleSpace=BCS_BoneSpace
    translationMode=BMM_Additive
    translationSpace=BCS_BoneSpace
    alphaBoolBlend {blendOption=Linear}
    boneToModify {boneName=dyn_R_forearmTwist_volume_2}
    scale {x=0.6}
    translation {x=-2}
   }
   n15 = "Transform (Modify) Bone - Bone: dyn_R_forearmTwist_volume_2"(ComponentPose=@n14, Alpha=Get R_wrist_rotate_Z() -> (R_wrist_rotate_Z:double), Translation="-2,0,0", Rotation="0,0,0", Scale="0.6,1,1") {
    rotationMode=BMM_Additive
    rotationSpace=BCS_BoneSpace
    scaleMode=BMM_Additive
    scaleSpace=BCS_BoneSpace
    translationMode=BMM_Additive
    translationSpace=BCS_BoneSpace
    alphaBoolBlend {blendOption=Linear}
    boneToModify {boneName=dyn_R_forearmTwist_volume_2}
    scale {x=0.6}
    translation {x=-2}
   }
   n16 = "Transform (Modify) Bone - Bone: dyn_L_forearmTwist_volume_2"(ComponentPose=@n15, Alpha=Get L_wrist_rotate_Y() -> (L_wrist_rotate_Y:double), Translation="2,0,0", Rotation="0,0,0", Scale="0.6,1,1") {
    rotationMode=BMM_Additive
    rotationSpace=BCS_BoneSpace
    scaleMode=BMM_Additive
    scaleSpace=BCS_BoneSpace
    translationMode=BMM_Additive
    translationSpace=BCS_BoneSpace
    alphaBoolBlend {blendOption=Linear}
    boneToModify {boneName=dyn_L_forearmTwist_volume_2}
    scale {x=0.6}
    translation {x=2}
   }
   n17 = "Transform (Modify) Bone - Bone: dyn_L_forearmTwist_volume_2"(ComponentPose=@n16, Alpha=Get L_wrist_rotate_Z() -> (L_wrist_rotate_Z:double), Translation="2,0,0", Rotation="0,0,0", Scale="0.6,1,1") {
    rotationMode=BMM_Additive
    rotationSpace=BCS_BoneSpace
    scaleMode=BMM_Additive
    scaleSpace=BCS_BoneSpace
    translationMode=BMM_Additive
    translationSpace=BCS_BoneSpace
    alphaBoolBlend {blendOption=Linear}
    boneToModify {boneName=dyn_L_forearmTwist_volume_2}
    scale {x=0.6}
    translation {x=2}
   }
   n18 = "Transform (Modify) Bone - Bone: dyn_L_forearmTwist_volume_2"(ComponentPose=@n17, Alpha=Get L_Wrist_Rotate_X() -> (L_Wrist_Rotate_X:double), Translation="1,0,0", Rotation="0,0,0", Scale="0.7,1,1") {
    rotationMode=BMM_Additive
    rotationSpace=BCS_BoneSpace
    scaleMode=BMM_Additive
    scaleSpace=BCS_BoneSpace
    translationMode=BMM_Additive
    translationSpace=BCS_BoneSpace
    alphaBoolBlend {blendOption=Linear}
    boneToModify {boneName=dyn_L_forearmTwist_volume_2}
    scale {x=0.7}
    translation {x=1}
   }
   n19 = "Transform (Modify) Bone - Bone: R_strapBack_A_jnt_follow"(ComponentPose=@n18, Alpha=0, Translation="-6,-3,0", Rotation="0,0,0", Scale="1,1,1") {
    translationMode=BMM_Additive
    translationSpace=BCS_BoneSpace
    alphaBoolBlend {blendOption=Linear}
    boneToModify {boneName=R_strapBack_A_jnt_follow}
    translation=<-6 -3>
   }
   n20 = "Transform (Modify) Bone - Bone: R_strapBack_B_jnt"(ComponentPose=@n19, Alpha=0, Translation="-0.5,0,0", Rotation="0,0,0", Scale="1,1,1") {
    translationMode=BMM_Additive
    translationSpace=BCS_BoneSpace
    alphaBoolBlend {blendOption=Linear}
    boneToModify {boneName=R_strapBack_B_jnt}
    translation {x=-0.5}
   }
   n21 = "Transform (Modify) Bone - Bone: L_strapBack_A_jnt_follow"(ComponentPose=@n20, Alpha=0, Translation="0,2,0", Rotation="0,0,0", Scale="1,1,1") {
    translationMode=BMM_Additive
    translationSpace=BCS_BoneSpace
    alphaBoolBlend {blendOption=Linear}
    boneToModify {boneName=L_strapBack_A_jnt_follow}
    translation {y=2}
   }
   n22 = "Transform (Modify) Bone - Bone: L_strapBack_B_jnt"(ComponentPose=@n21, Alpha=1, Translation="1,0,0", Rotation="0,0,0", Scale="1,1,1") {
    translationMode=BMM_Additive
    translationSpace=BCS_BoneSpace
    alphaBoolBlend {blendOption=Linear}
    boneToModify {boneName=L_strapBack_B_jnt}
    translation {x=1}
   }
   n23 = "Transform (Modify) Bone - Bone: R_strapBack_A_jnt_follow"(ComponentPose=@n22, Alpha=Get R_upperarm_atRest() -> (R_upperarm_atRest:double), Translation="-12,-12,0", Rotation="0,0,0", Scale="1,1,1") {
    translationMode=BMM_Additive
    translationSpace=BCS_BoneSpace
    alphaBoolBlend {blendOption=Linear}
    boneToModify {boneName=R_strapBack_A_jnt_follow}
    translation=<-12 -12>
   }
   n24 = "Transform (Modify) Bone - Bone: L_strapBack_A_jnt_follow"(ComponentPose=@n23, Alpha=Get L_upperarm_atRest() -> (L_upperarm_atRest:double), Translation="17,14,0", Rotation="0,0,0", Scale="1,1,1") {
    translationMode=BMM_Additive
    translationSpace=BCS_BoneSpace
    alphaBoolBlend {blendOption=Linear}
    boneToModify {boneName=L_strapBack_A_jnt_follow}
    translation=<17 14>
   }
   n25 = RigidBody(ComponentPose=@n24) {
    bClampLinearTranslationLimitToRefPose=true
    bOverrideWorldGravity=true
    overridePhysicsAsset=&/F_MED_Spartan_Future_Physics
    simulationSpace=BaseBoneSpace
    alphaBoolBlend {blendOption=Linear}
    baseBoneRef {boneName=pelvis}
    componentAppliedLinearAccClamp=<4000 4000 4000>
    componentLinearAccScale=<0.3 0.3 0.4>
    componentLinearVelScale=<0.25 0.25 0.25>
    overrideWorldGravity {z=-800}
   }
   n26 = Get Crouching_float() -> (Crouching_float:double)
   n27 = "Transform (Modify) Bone - Bone: dyn_sidePouch_jnt"(ComponentPose=@n25, Alpha=@n26, Translation="2,-5.5,0", Rotation="0,-5,-5", Scale="1,1,1") {
    rotationMode=BMM_Additive
    rotationSpace=BCS_BoneSpace
    scaleSpace=BCS_BoneSpace
    translationMode=BMM_Additive
    translationSpace=BCS_BoneSpace
    alphaBoolBlend {blendOption=Linear}
    alphaScaleBiasClamp {bInterpResult=true, interpSpeedDecreasing=20, interpSpeedIncreasing=20}
    boneToModify {boneName=dyn_sidePouch_jnt}
    rotation {roll=-5, yaw=-5}
    translation=<2 -5.5>
   }
   n28 = "Transform (Modify) Bone - Bone: dyn_beltLeftThigh_jnt"(ComponentPose=@n27, Alpha=@n26, Translation="0,2,8", Rotation="0,0,-60", Scale="1,1,1") {
    rotationMode=BMM_Additive
    rotationSpace=BCS_BoneSpace
    scaleSpace=BCS_BoneSpace
    translationMode=BMM_Additive
    translationSpace=BCS_BoneSpace
    alphaBoolBlend {blendOption=Linear}
    alphaScaleBiasClamp {bInterpResult=true, interpSpeedDecreasing=20, interpSpeedIncreasing=20}
    boneToModify {boneName=dyn_beltLeftThigh_jnt}
    rotation {roll=-60}
    translation {y=2, z=8}
   }
   n29 = "Constraint - dyn_bullet_jnt_1 (PPPPP)"!(
    ComponentPose=@n28
    Alpha=1
    ConstraintWeights_0=0
    ConstraintWeights_1:float=0.0508
    ConstraintWeights_2:float=0.0562
    ConstraintWeights_3:float=0.006
    ConstraintWeights_4:float=0.887
   ) -> (Pose) {
    alphaBoolBlend {blendOption=Linear}
    boneToModify {boneName=dyn_bullet_jnt_1}
    constraintSetup [
     {targetBone {boneName=dyn_beltLeftThigh_jnt}}
     {targetBone {boneName=pelvis}}
     {targetBone {boneName=dyn_beltSide_jnt}}
     {targetBone {boneName=dyn_beltFront_jnt}}
     {targetBone {boneName=dyn_beltBack_jnt}}
    ]
    constraintWeights [1,1,1,1,1]
   }
   n30 = "Constraint - dyn_bullet_jnt_2 (PPPPP)"!(
    ComponentPose=@n29
    Alpha=1
    ConstraintWeights_0=0
    ConstraintWeights_1:float=0.0418
    ConstraintWeights_2:float=0.1351
    ConstraintWeights_3:float=0.0082
    ConstraintWeights_4:float=0.8149
   ) -> (Pose) {
    alphaBoolBlend {blendOption=Linear}
    boneToModify {boneName=dyn_bullet_jnt_2}
    constraintSetup [
     {targetBone {boneName=dyn_beltLeftThigh_jnt}}
     {targetBone {boneName=pelvis}}
     {targetBone {boneName=dyn_beltSide_jnt}}
     {targetBone {boneName=dyn_beltFront_jnt}}
     {targetBone {boneName=dyn_beltBack_jnt}}
    ]
    constraintWeights [1,1,1,1,1]
   }
   n31 = "Constraint - dyn_bullet_jnt_3 (PPPPP)"!(
    ComponentPose=@n30
    Alpha=1
    ConstraintWeights_0=0
    ConstraintWeights_1:float=0.0427
    ConstraintWeights_2:float=0.2707
    ConstraintWeights_3:float=0.011
    ConstraintWeights_4:float=0.6757
   ) -> (Pose) {
    alphaBoolBlend {blendOption=Linear}
    boneToModify {boneName=dyn_bullet_jnt_3}
    constraintSetup [
     {targetBone {boneName=dyn_beltLeftThigh_jnt}}
     {targetBone {boneName=pelvis}}
     {targetBone {boneName=dyn_beltSide_jnt}}
     {targetBone {boneName=dyn_beltFront_jnt}}
     {targetBone {boneName=dyn_beltBack_jnt}}
    ]
    constraintWeights [1,1,1,1,1]
   }
   n32 = "Constraint - dyn_bullet_jnt_4 (PPPPP)"!(
    ComponentPose=@n31
    Alpha=1
    ConstraintWeights_0=0
    ConstraintWeights_1:float=0.0429
    ConstraintWeights_2:float=0.4539
    ConstraintWeights_3:float=0.018
    ConstraintWeights_4:float=0.4852
   ) -> (Pose) {
    alphaBoolBlend {blendOption=Linear}
    boneToModify {boneName=dyn_bullet_jnt_4}
    constraintSetup [
     {targetBone {boneName=dyn_beltLeftThigh_jnt}}
     {targetBone {boneName=pelvis}}
     {targetBone {boneName=dyn_beltSide_jnt}}
     {targetBone {boneName=dyn_beltFront_jnt}}
     {targetBone {boneName=dyn_beltBack_jnt}}
    ]
    constraintWeights [1,1,1,1,1]
   }
   n33 = "Constraint - dyn_bullet_jnt_5 (PPPPP)"!(
    ComponentPose=@n32
    Alpha=1
    ConstraintWeights_0=0
    ConstraintWeights_1:float=0.0454
    ConstraintWeights_2:float=0.6203
    ConstraintWeights_3:float=0.036
    ConstraintWeights_4:float=0.2983
   ) -> (Pose) {
    alphaBoolBlend {blendOption=Linear}
    boneToModify {boneName=dyn_bullet_jnt_5}
    constraintSetup [
     {targetBone {boneName=dyn_beltLeftThigh_jnt}}
     {targetBone {boneName=pelvis}}
     {targetBone {boneName=dyn_beltSide_jnt}}
     {targetBone {boneName=dyn_beltFront_jnt}}
     {targetBone {boneName=dyn_beltBack_jnt}}
    ]
    constraintWeights [1,1,1,1,1]
   }
   n34 = "Constraint - dyn_bullet_jnt_6 (PPPPP)"!(
    ComponentPose=@n33
    Alpha=1
    ConstraintWeights_0=0.0001
    ConstraintWeights_1:float=0.0644
    ConstraintWeights_2:float=0.6001
    ConstraintWeights_3:float=0.1437
    ConstraintWeights_4:float=0.1917
   ) -> (Pose) {
    alphaBoolBlend {blendOption=Linear}
    boneToModify {boneName=dyn_bullet_jnt_6}
    constraintSetup [
     {targetBone {boneName=dyn_beltLeftThigh_jnt}}
     {targetBone {boneName=pelvis}}
     {targetBone {boneName=dyn_beltSide_jnt}}
     {targetBone {boneName=dyn_beltFront_jnt}}
     {targetBone {boneName=dyn_beltBack_jnt}}
    ]
    constraintWeights [1,1,1,1,1]
   }
   n35 = "Constraint - dyn_bullet_jnt_7 (PPPPP)"!(
    ComponentPose=@n34
    Alpha=1
    ConstraintWeights_0=0.0003
    ConstraintWeights_1:float=0.0389
    ConstraintWeights_2:float=0.3327
    ConstraintWeights_3:float=0.5307
    ConstraintWeights_4:float=0.0974
   ) -> (Pose) {
    alphaBoolBlend {blendOption=Linear}
    boneToModify {boneName=dyn_bullet_jnt_7}
    constraintSetup [
     {targetBone {boneName=dyn_beltLeftThigh_jnt}}
     {targetBone {boneName=pelvis}}
     {targetBone {boneName=dyn_beltSide_jnt}}
     {targetBone {boneName=dyn_beltFront_jnt}}
     {targetBone {boneName=dyn_beltBack_jnt}}
    ]
    constraintWeights [1,1,1,1,1]
   }
   n36 = "Constraint - dyn_bullet_jnt_8 (PPPPP)"!(
    ComponentPose=@n35
    Alpha=1
    ConstraintWeights_0=0.0004
    ConstraintWeights_1:float=0.0044
    ConstraintWeights_2:float=0.0345
    ConstraintWeights_3:float=0.9509
    ConstraintWeights_4:float=0.0097
   ) -> (Pose) {
    alphaBoolBlend {blendOption=Linear}
    boneToModify {boneName=dyn_bullet_jnt_8}
    constraintSetup [
     {targetBone {boneName=dyn_beltLeftThigh_jnt}}
     {targetBone {boneName=pelvis}}
     {targetBone {boneName=dyn_beltSide_jnt}}
     {targetBone {boneName=dyn_beltFront_jnt}}
     {targetBone {boneName=dyn_beltBack_jnt}}
    ]
    constraintWeights [1,1,1,1,1]
   }
   n37 = "Constraint - dyn_bullet_jnt_9 (PPPPP)"!(
    ComponentPose=@n36
    Alpha=1
    ConstraintWeights_0=0.0022
    ConstraintWeights_1:float=0.0021
    ConstraintWeights_2:float=0.0041
    ConstraintWeights_3:float=0.9905
    ConstraintWeights_4:float=0.0011
   ) -> (Pose) {
    alphaBoolBlend {blendOption=Linear}
    boneToModify {boneName=dyn_bullet_jnt_9}
    constraintSetup [
     {targetBone {boneName=dyn_beltLeftThigh_jnt}}
     {targetBone {boneName=pelvis}}
     {targetBone {boneName=dyn_beltSide_jnt}}
     {targetBone {boneName=dyn_beltFront_jnt}}
     {targetBone {boneName=dyn_beltBack_jnt}}
    ]
    constraintWeights [1,1,1,1,1]
   }
   n38 = "Constraint - dyn_bullet_jnt_10 (PPPPP)"!(
    ComponentPose=@n37
    Alpha=1
    ConstraintWeights_0=0.0265
    ConstraintWeights_1:float=0.0239
    ConstraintWeights_2:float=0.0326
    ConstraintWeights_3:float=0.9095
    ConstraintWeights_4:float=0.0075
   ) -> (Pose) {
    alphaBoolBlend {blendOption=Linear}
    boneToModify {boneName=dyn_bullet_jnt_10}
    constraintSetup [
     {targetBone {boneName=dyn_beltLeftThigh_jnt}}
     {targetBone {boneName=pelvis}}
     {targetBone {boneName=dyn_beltSide_jnt}}
     {targetBone {boneName=dyn_beltFront_jnt}}
     {targetBone {boneName=dyn_beltBack_jnt}}
    ]
    constraintWeights [1,1,1,1,1]
   }
   n39 = Control Rig(Source=Component To Local(ComponentPose=@n38)) {
    alphaBoolBlend {blendOption=Linear}
    controlRigAssetReference {blueprintRigClass=&/F_MED_Spartan_Future_CtrlRig.F_MED_Spartan_Future_CtrlRig_C}
   }
   Output Pose(Result=@n39) {layerGroup=None}
  ]}
 EventGraph {nodes [
   Event BlueprintUpdateAnimation()
   > n1 = Sequence!(execute) -> (then_0, then_1, then_2:Exec, then_3:Exec, then_4:Exec, then_5:Exec, then_6:Exec, then_7:Exec, then_8:Exec)
   n2 = GetOwnerSkeletalMeshComponent()
   n3 = BreakRotator(InRot=BreakTransform(InTransform=GetDeltaTransformFromRefPose(self=@n2, BoneName=hand_r, BaseName=lowerarm_r)).Rotation)
   > Set R_wrist_rotate_Z(R_wrist_rotate_Z:double=MapRangeClamped(Value=@n3.Yaw, InRangeA=0, InRangeB=30)) -> (Output_Get:double)
   n4 = BreakTransform(InTransform=GetDeltaTransformFromRefPose(self=GetOwnerSkeletalMeshComponent(), BoneName=lowerarm_r, BaseName=upperarm_r))
   Set R_forearm_ADS(execute=@n1.then_1, R_forearm_ADS:double=MapRangeClamped(Value=BreakRotator(InRot=@n4.Rotation).Yaw, InRangeA=40, InRangeB=60)) -> (Output_Get:double)
   n5 = BreakTransform(InTransform=GetDeltaTransformFromRefPose(self=GetOwnerSkeletalMeshComponent(), BoneName=lowerarm_l, BaseName=upperarm_l))
   Set L_forearm_ADS(execute=@n1.then_2, L_forearm_ADS:double=MapRangeClamped(Value=BreakRotator(InRot=@n5.Rotation).Yaw, InRangeA=35, InRangeB=60)) -> (Output_Get:double)
   Set R_wrist_rotate_Y(execute=@n1.then_3, R_wrist_rotate_Y:double=MapRangeClamped(Value=@n3.Pitch, InRangeA=-20, InRangeB=50)) -> (Output_Get:double)
   n6 = GetOwnerSkeletalMeshComponent()
   n7 = BreakRotator(InRot=BreakTransform(InTransform=GetDeltaTransformFromRefPose(self=@n6, BoneName=hand_l, BaseName=lowerarm_l)).Rotation)
   Set L_wrist_rotate_Y(execute=@n1.then_4, L_wrist_rotate_Y:double=MapRangeClamped(Value=@n7.Pitch, InRangeA=-20, InRangeB=50)) -> (Output_Get:double)
   Set L_wrist_rotate_Z(execute=@n1.then_5, L_wrist_rotate_Z:double=MapRangeClamped(Value=@n7.Yaw, InRangeA=0, InRangeB=30)) -> (Output_Get:double)
   Set L_Wrist_Rotate_X(execute=@n1.then_6, L_Wrist_Rotate_X:double=MapRangeClamped(Value=@n7, InRangeA=-60, InRangeB=-100)) -> (Output_Get:double)
   n8 = BreakTransform(InTransform=GetDeltaTransformFromRefPose(self=GetOwnerSkeletalMeshComponent(), BoneName=upperarm_r, BaseName=clavicle_r))
   Set R_upperarm_atRest(execute=@n1.then_7, R_upperarm_atRest:double=MapRangeClamped(Value=BreakRotator(InRot=@n8.Rotation).Yaw, InRangeA=5, InRangeB=-5)) -> (Output_Get:double)
   n9 = BreakTransform(InTransform=GetDeltaTransformFromRefPose(self=GetOwnerSkeletalMeshComponent(), BoneName=upperarm_l, BaseName=clavicle_l))
   Set L_upperarm_atRest(execute=@n1.then_8, L_upperarm_atRest:double=MapRangeClamped(Value=BreakRotator(InRot=@n9.Rotation).Yaw, InRangeA=5, InRangeB=-5)) -> (Output_Get:double)
   Event OnBeginCrouching()
   > Set Crouching_float(Crouching_float:double=1) -> (Output_Get:double)
   Event OnEndCrouching()
   > Set Crouching_float(Crouching_float=0)
  ]}
]
variables [
 Crouching_float {type=number}
 L_forearm_ADS {type=number}
 L_upperarm_atRest {type=number}
 L_Wrist_Rotate_X {type=number}
 L_wrist_rotate_Y {type=number}
 L_wrist_rotate_Z {type=number}
 R_forearm_ADS {type=number}
 R_upperarm_atRest {type=number}
 R_wrist_rotate_Y {type=number}
 R_wrist_rotate_Z {type=number}
 Skel_Mesh {type=SkeletalMeshComponent}
]
