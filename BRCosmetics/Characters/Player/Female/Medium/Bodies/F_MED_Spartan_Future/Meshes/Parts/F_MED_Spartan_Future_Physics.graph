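# Physics asset for the F_MED_Spartan_Future female medium player body, set up for the left and right dynamic hood bones.
# Capsule bodies cover two bones per hood side, plus the head and upper spine (spine_05).
# The first hood bones have tight swing and twist limits to spine_05, while the second pair attach to the head with limited swing and locked twist.
# The head rotates freely relative to spine_05.
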PhysicsAsset /BRCosmetics/Characters/Player/Female/Medium/Bodies/F_MED_Spartan_Future/Meshes/Parts/F_MED_Spartan_Future_Physics
bodies {
 dyn_L_hood [{shapeType=Capsule, length=10, radius=3, center=<0.3914 0 0>, rotation=<rot -7.6208 0 0>}]
 dyn_L_hood_2 [{shapeType=Capsule, length=5, radius=2, center=<0 0.0001 -1.1064>}]
 dyn_R_hood [{shapeType=Capsule, length=7, radius=3, center=<0.3914 0 0>, rotation=<rot -7.6208 0 0>}]
 dyn_R_hood_2 [{shapeType=Capsule, length=5, radius=2, center=<0 0.0001 -1.1064>}]
 head [{shapeType=Capsule, length=13.7553, radius=7.5, rotation=<rot 90 0 -15.5351>}]
 spine_05 [{shapeType=Capsule, length=11.3488, radius=7.3}]
}
constraints [
 {bone1Name=dyn_L_hood, bone2Name=spine_05, swing1LimitDegrees=5, swing2LimitDegrees=5, twistLimitDegrees=5}
 {bone1Name=dyn_R_hood, bone2Name=spine_05, swing1LimitDegrees=5, swing2LimitDegrees=5, twistLimitDegrees=5}
 {bone1Name=dyn_R_hood_2, bone2Name=head, swing1LimitDegrees=10, swing2LimitDegrees=10, twistLimitDegrees=45, twistMotion=Locked}
 {bone1Name=dyn_L_hood_2, bone2Name=head, swing1LimitDegrees=10, swing2LimitDegrees=10, twistLimitDegrees=45, twistMotion=Locked}
 {
  bone1Name=head
  bone2Name=spine_05
  swing1LimitDegrees=45
  swing1Motion=Free
  swing2LimitDegrees=45
  swing2Motion=Free
  twistLimitDegrees=45
  twistMotion=Free
 }
]
